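# Sound cue for the thrust start of the TwentyTwoMale cosmetic glider.
# A mixer layers a looping thrust wave with one of three randomly chosen, equally weighted one-shot variations.
# Playback runs at reduced volume through shared glider thrust attenuation and concurrency settings, under the glider landing sound class.
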SoundCue /BRCosmetics/Sounds/Gliders/TwentyTwoMale/Glider_TwentyTwoMale_ThrustStart_Cue
root=Mixer_0
nodes [
 {id=Mixer_0, children [Random_0,WavePlayer_0], properties {inputVolume [1,1]}}
 {id=Random_0, children [WavePlayer_1,WavePlayer_2,WavePlayer_3], properties {weights [1,1,1]}}
 {id=WavePlayer_0, properties {bLooping=true, soundWaveAssetPtr=&/Glider_TwentyTwoMale_ThrustStart}}
 {id=WavePlayer_1, properties {soundWaveAssetPtr=&/Glider_TwentyTwoMale_ThrustStart_OneShot_01}}
 {id=WavePlayer_2, properties {soundWaveAssetPtr=&/Glider_TwentyTwoMale_ThrustStart_OneShot_02}}
 {id=WavePlayer_3, properties {soundWaveAssetPtr=&/Glider_TwentyTwoMale_ThrustStart_OneShot_03}}
]
properties {
 attenuationSettings=&Fort_Player_Glider_Thrust_Athena_Att
 maxDistance=12500
 soundClassObject=&Fort_Gliders_Land
 volumeMultiplier=0.5
 concurrencySet [&Athena_Glider_Thrust_Conc]
}
